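# Niagara system for the petal effects of the MuzzleSwish companion emote, tinted per variant through User.VariantColorIndex.
# Two Jump emitters burst mesh and sprite layers split by spawn group, with per-group lifetimes and tumbling drag.
# Petals fall under gravity, wind and a timed vortex, colliding only on certain platforms or in front-end preview.
# Reveal001 layers quality-scaled sprites around a shape, with curl noise, late-loop vortex and distance-based alpha.
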NiagaraSystem /CosmeticCompanions/FX/Builtins/MuzzleSwish/Emote/FX/NS_Emote_MuzzleSwish_Petals
dependencies {
 usedDataInterfaces [
  &NiagaraDataInterfaceCurve
  &NiagaraDataInterfacePlatformSet
  &NiagaraDataInterfaceArrayFloat
  &NiagaraDataInterfaceArrayColor
  &NiagaraDataInterfaceColorCurve
 ]
 usedDynamicInputs []
 usedModules [
  &SystemState
  &self:SystemSpawnScript.SystemScriptSource.SystemScriptGraph.NiagaraNodeAssignment_0.SetVariables_#1
  &self:SystemSpawnScript.SystemScriptSource.SystemScriptGraph.NiagaraNodeAssignment_1.SetVariables_#2
  &self:SystemSpawnScript.SystemScriptSource.SystemScriptGraph.NiagaraNodeAssignment_2.SetVariables_#3
 ]
}
emitters {
 JumpBig002 {
  emitterSpawnScript [/SetVariables(Emitter.RotationSpeed:NiagaraFloat=&FloatFromCurve(FloatCurve=DataInterface() {propertyValues {Curve {keys [<at 0 -0.5>,<at 0.1 0>]}}}, CurveIndex=$Emitter.NormalizedLoopAge, Scale Curve=0)) {isSetParameters=true, module=SetVariables_#4}]
  emitterUpdateScript [
   &EmitterState(Inactive Response=::NewEnumerator1, Life Cycle Mode=::NewEnumerator1, Loop Behavior=::NewEnumerator1, Loop Duration=1.6, UseLoopDelay=true)
   /SetVariables(
    Emitter.DecalLifeTime=0.5
    Emitter.MainRingLifeTime=0.45
    Emitter.Petals:NiagaraFloat=&RandomRangeFloat(Minimum=1, Maximum=1.5)
    Emitter.WindBurstLifeTime:NiagaraFloat=&RandomRangeFloat(Minimum=0.5, Maximum=0.8)
   ) {isSetParameters=true, module=SetVariables_#5}
   &SpawnBurst_Instantaneous(Use Spawn Probability=false, UseLoopCountLimit=false)
   &SpawnBurst_Instantaneous(Spawn Group=1, Use Spawn Probability=false, UseLoopCountLimit=false) {module=SpawnBurst_Instantaneous001}
   &SpawnBurst_Instantaneous(Spawn Group=2, Spawn Time=0.1, Use Spawn Probability=false, UseLoopCountLimit=false) {module=SpawnBurst_Instantaneous002}
   &SpawnBurst_Instantaneous(Spawn Count=20, Spawn Group=3, Spawn Time=0.1, Use Spawn Probability=false, UseLoopCountLimit=false) {module=SpawnBurst_Instantaneous003}
  ]
  particleSpawnScript [
   /SetVariables(Particles.VisibilityTag=&SpawnGroup()) {isSetParameters=true, module=SetVariables_#6}
   &InitializeParticle(
    Color=&SelectLinearColorFromArray(
     Array Sampling Mode=::NewEnumerator1
     Direct Array Index=$User.VariantColorIndex
     Color Selection Array=DataInterface() {propertyValues {ColorData [<rgba 3 1.4633 2.7782 0.1>, <rgba 0.3531 0.2918 0.4583 1>, <rgba 0 0.3501 1 1>]}}
     Random Seed=?
    )
    Lifetime=&MergeHelper_FloatSwitch(
     0=$Emitter.WindBurstLifeTime
     1=$Emitter.MainRingLifeTime
     2=$Emitter.DecalLifeTime
     3=$Emitter.Petals
     Index=&SpawnGroup()
     Num=::NewEnumerator2
    )
    Mesh Scale=&MergeHelper_VectorSwitch(
     0=&Multiply_VectorByFloat(Vector=<1.15 1.15 0.2>, Float=&RandomRangeFloat(Minimum=0.25, Maximum=0.375))
     1=<0.07 0.1 0.07>
     2=?
     3=?
     Index=&SpawnGroup()
     Num=::NewEnumerator0
    )
    Mesh Scale Mode=::NewEnumerator4
    Position Offset=<-21 9 0>
    Position Offset Coordinate Space=::Local
    Sprite Size Mode=::NewEnumerator3
    Uniform Sprite Size=35
    UsePositionOffset=true
    Write Material Random=true
   )
   &InitialMeshOrientation(
    Mesh Axis To Orient=<0 0 3>
    Mesh Orientation Mode=::NewEnumerator2
    Orientation Coordinate Space=::World
    Orientation Vector=&MergeHelper_VectorSwitch(0=<0 0 1>, 1=<0 1 0>, 2=?, 3=?, Index=&SpawnGroup(), Num=::NewEnumerator0)
    Rotation=&MergeHelper_VectorSwitch(
     0=&RandomRangeVector(Minimum=<0.03 0.03 0>, Maximum=<-0.03 -0.03 1>)
     1=&RandomRangeVector(Minimum=<0 0 0>, Maximum=<0 1 0>)
     2=?
     3=?
     Index=&SpawnGroup()
     Num=::NewEnumerator0
    )
    Use Rotation Vector=true
   )
   /SetVariables(Particles.DecalFacing:Vector3f=<0 0 1>) {isSetParameters=true, module=SetVariables_#7}
   &AddVelocity(
    Cone Axis=<0.1 0 1>
    Module.UseConeFalloff=false
    Module.UseVelDistribution=false
    Override Seed=false
    Velocity Mode=::NewEnumerator2
    Velocity Speed=&MaskFloatBySpawnGroup(Masked Float=&RandomRangeFloat(Minimum=50, Maximum=150), Passthrough Float=1, Spawn Group Mask=3)
   )
   &ShapeLocation(
    Cylinder Height=0
    Cylinder Radius=18
    MaskSpawnGroup=true
    Override Seed=false
    Shape Primitive=::NewEnumerator1
    Spawn Group Mask=3
    Surface Only=false
    UseOwnerScale=true
   )
  ]
  particleUpdateScript [
   &ParticleState()
   &ScaleSpriteSize(Uniform Curve Sprite Scale=DataInterface() {propertyValues {Curve {
       defaultValue=0
       keys [
        {arriveTangent=4.1785, interpMode=RCIM_Cubic, leaveTangent=4.1785, tangentMode=RCTM_User, time=0, value=0}
        {arriveTangent=1.1179, interpMode=RCIM_Cubic, leaveTangent=1.1179, tangentMode=RCTM_User, time=0.2042, value=0.6228}
        {arriveTangent=0.0224, interpMode=RCIM_Cubic, leaveTangent=0.0224, tangentMode=RCTM_User, time=1, value=1}
       ]
      }}})
   &ScaleMeshSize(Scale Factor=&MergeHelper_VectorSwitch(
     0=&VectorFromFloat(Value=&FloatFromCurve(FloatCurve=DataInterface() {propertyValues {Curve {keys [<at 0 0.2144 RCIM_Cubic>,<at 1 0.7306>]}}}, Scale Curve=1.4129))
     1=&VectorFromFloat(Value=&FloatFromCurve(
       FloatCurve=DataInterface() {propertyValues {Curve {
          defaultValue=0.4154
          keys [
           {arriveTangent=0.8683, interpMode=RCIM_Cubic, leaveTangent=0.8683, tangentMode=RCTM_User, time=0, value=0}
           {arriveTangent=1.1179, interpMode=RCIM_Cubic, leaveTangent=1.1179, tangentMode=RCTM_User, time=0.2125, value=0.1276}
           {arriveTangent=1.3888, interpMode=RCIM_Cubic, leaveTangent=1.3888, time=0.297, value=0.6941}
           {arriveTangent=0.0224, interpMode=RCIM_Cubic, leaveTangent=0.0224, tangentMode=RCTM_User, time=0.8406, value=1}
          ]
         }}}
       Scale Curve=1.829
      ))
     2=?
     3=?
     Index=&SpawnGroup()
     Num=::NewEnumerator0
    ))
   &UpdateMeshOrientation(
    Rotation Rate=&FloatFromCurve(
     FloatCurve=DataInterface() {propertyValues {Curve {keys [<at 0.55 0.4376 RCIM_Cubic>, {arriveTangent=2.9123, interpMode=RCIM_Cubic, leaveTangent=2.9123, tangentMode=RCTM_User, time=1, value=1}]}}}
     CurveIndex=$Emitter.NormalizedLoopAge
     Scale Curve=-1.6154
    )
    Rotation Vector=&MergeHelper_VectorSwitch(0=<0 0 1>, 1=<0 3 0>, 2=?, 3=?, Index=&SpawnGroup(), Num=::NewEnumerator0)
   )
   &ScaleColor(
    Scale Alpha=&FloatFromCurve(
     FloatCurve=DataInterface() {propertyValues {Curve {keys [
         <at 0 0 RCIM_Cubic>
         {interpMode=RCIM_Cubic, tangentMode=RCTM_User, time=0.1328, value=1}
         {interpMode=RCIM_Cubic, tangentMode=RCTM_User, time=0.6888, value=1}
         {tangentMode=RCTM_User, time=1, value=0}
        ]}}}
     Scale Curve=&Multiply_Float(A=0.09, B=&FNTimeOfDayScalar(Night Scale=0.7, Morning Scale=0.9, Evening Scale=0.8))
    )
    Scale RGB=<2 2 2>
    ScaleA=true
    ScaleRGB=true
   )
   &DynamicMaterialParameters(
    Index 0 Param 1=&MergeHelper_FloatSwitch(
     0=&FloatFromCurve(FloatCurve=DataInterface() {propertyValues {Curve {keys [{tangentMode=RCTM_User, time=0.4, value=1}, {tangentMode=RCTM_User, time=0.9, value=-1}]}}})
     1=&FloatFromCurve(FloatCurve=DataInterface() {propertyValues {Curve {keys [
          {arriveTangent=2.4536, interpMode=RCIM_Cubic, leaveTangent=2.4536, tangentMode=RCTM_User, time=0, value=0}
          {interpMode=RCIM_Cubic, tangentMode=RCTM_User, time=0.5, value=0.5}
          {arriveTangent=-2.5865, tangentMode=RCTM_User, time=1, value=0}
         ]}}})
     2=?
     3=?
     Index=&SpawnGroup()
     Num=::NewEnumerator0
    )
    Index 0 Param 2=1
    Index 0 Param 3=&RandomRangeFloat(Minimum=-1, Maximum=1)
    Index 0 Param 4=&RandomRangeFloat(Minimum=-1, Maximum=1)
    Index 1 Param 1=&FloatFromCurve(
     FloatCurve=DataInterface() {propertyValues {Curve {keys [
         {arriveTangent=1.7366, interpMode=RCIM_Cubic, leaveTangent=1.7366, tangentMode=RCTM_User, time=0, value=0}
         {arriveTangent=0.4174, interpMode=RCIM_Cubic, leaveTangent=0.4174, tangentMode=RCTM_User, time=1, value=1}
        ]}}}
     Scale Curve=0.5
    )
    Index 1 Param 2=1
    Module.Index0Float0Write=true
    Module.Index0Float1Write=true
    Module.Index0Float2Write=true
    Module.Index0Float3Write=true
    Module.Index1Float0Write=true
    Module.Index1Float1Write=false
    Module.Index1Float2Write=false
    Module.Index1Float3Write=false
    Write Parameter Index 1=::NewEnumerator1
   ) {module=DynamicMaterialParameters001}
   /SetVariables(Particles.PetalSize:Vector2f=&Multiply_Vector2D_ByFloat(
     Float=&FloatFromCurve(
      FloatCurve=DataInterface() {propertyValues {Curve {
         defaultValue=1
         keys [
          {arriveTangent=14.5835, interpMode=RCIM_Cubic, leaveTangent=14.5835, tangentMode=RCTM_User, time=0, value=0}
          {arriveTangent=1.1111, interpMode=RCIM_Cubic, leaveTangent=1.1111, time=0.1, value=1}
          {arriveTangent=-1.1111, interpMode=RCIM_Cubic, leaveTangent=-1.1111, time=0.9, value=1}
          {arriveTangent=-8.7113, arriveTangentWeight=0.0333, interpMode=RCIM_Cubic, leaveTangent=-8.7113, tangentMode=RCTM_User, time=1, value=0}
         ]
        }}}
      Scale Curve=&RandomRangeFloat(Minimum=2, Maximum=3)
     )
     Vector2D=<1 1>
    )) {isSetParameters=true, module=SetVariables_#8}
   &AerodynamicDrag(
    Aerodynamic Drag:NiagaraFloat=&RandomRangeFloat(Minimum=0.3, Maximum=0.5)
    Aerodynamic Rotational Drag=0.15
    Drag Shape Curve=DataInterface() {propertyValues {Curve {keys [
        {interpMode=RCIM_Cubic, tangentMode=RCTM_User, time=0, value=0.1}
        {arriveTangent=2.0799, interpMode=RCIM_Cubic, leaveTangent=2.0799, tangentMode=RCTM_User, time=1, value=1}
       ]}}}
    Lift Contribution=0.75
    Lift Shape Curve=DataInterface() {propertyValues {Curve {keys [
        {interpMode=RCIM_Cubic, tangentMode=RCTM_User, time=0, value=0}
        {interpMode=RCIM_Cubic, tangentMode=RCTM_User, time=0.5, value=1}
        {tangentMode=RCTM_User, time=1, value=0}
       ]}}}
    Pivot Offset Maximum:Vector3f=<1 0.5 1>
    Pivot Offset Minimum:Vector3f=<-1 -1 -1>
    Rotation Strength:NiagaraFloat=&RandomRangeFloat(Minimum=1.8, Maximum=3.2)
    Use Linear Drag=true
    Use Rotational Drag=true
   )
   &SolveForcesAndVelocity(Clamp Velocity=false, Limit Acceleration=false)
   &AlignSpriteToMeshOrientation(SetAlignment=true, SetFacing=true)
  ]
  rendererClasses [&NiagaraMeshRendererProperties, &NiagaraSpriteRendererProperties]
  renderers [
   {rendererClass=&NiagaraMeshRendererProperties, rendererIndex=0}
   {rendererClass=&NiagaraMeshRendererProperties, rendererIndex=1}
   {rendererClass=&NiagaraSpriteRendererProperties, rendererIndex=2}
   {rendererClass=&NiagaraSpriteRendererProperties, rendererIndex=3}
  ]
 }
 JumpSmall001 {
  emitterSpawnScript [/SetVariables(Emitter.RotationSpeed:NiagaraFloat=&FloatFromCurve(FloatCurve=DataInterface() {propertyValues {Curve {keys [<at 0 -0.5>,<at 0.1 0>]}}}, CurveIndex=$Emitter.NormalizedLoopAge, Scale Curve=0)) {isSetParameters=true, module=SetVariables_#4}]
  emitterUpdateScript [
   &EmitterState(
    Inactive Response=::NewEnumerator1
    Life Cycle Mode=::NewEnumerator1
    Loop Behavior=::NewEnumerator1
    Loop Delay=0.1
    Loop Duration=1.6
    UseLoopDelay=true
   )
   /SetVariables(
    Emitter.DecalLifeTime=0.5
    Emitter.MainRingLifeTime=0.45
    Emitter.Petals:NiagaraFloat=&RandomRangeFloat(Minimum=1, Maximum=1.5)
    Emitter.WindBurstLifeTime:NiagaraFloat=&RandomRangeFloat(Minimum=0.5, Maximum=0.8)
   ) {isSetParameters=true, module=SetVariables_#5}
   &SpawnBurst_Instantaneous(Spawn Count=2, Use Spawn Probability=false, UseLoopCountLimit=false)
   &SpawnBurst_Instantaneous(Spawn Count=2, Spawn Group=1, Use Spawn Probability=false, UseLoopCountLimit=false) {module=SpawnBurst_Instantaneous001}
   &SpawnBurst_Instantaneous(Spawn Count=2, Spawn Group=2, Spawn Time=0.1, Use Spawn Probability=false, UseLoopCountLimit=false) {module=SpawnBurst_Instantaneous002}
   &SpawnBurst_Instantaneous(Spawn Count=24, Spawn Group=3, Spawn Time=0.1, Use Spawn Probability=false, UseLoopCountLimit=false) {module=SpawnBurst_Instantaneous003}
  ]
  particleSpawnScript [
   &PartitionParticles()
   /SetVariables(Particles.VisibilityTag=&SpawnGroup()) {isSetParameters=true, module=SetVariables_#6}
   &InitializeParticle(
    Color=&SelectLinearColorFromArray(
     Array Sampling Mode=::NewEnumerator1
     Direct Array Index=$User.VariantColorIndex
     Color Selection Array=DataInterface() {propertyValues {ColorData [<rgba 3 1.4633 2.7782 0.1>, <rgba 0.3531 0.2918 0.4583 1>, <rgba 0.2051 0.3467 0.7529 1>]}}
     Random Seed=?
    )
    Lifetime=&MergeHelper_FloatSwitch(
     0=$Emitter.WindBurstLifeTime
     1=$Emitter.MainRingLifeTime
     2=$Emitter.DecalLifeTime
     3=$Emitter.Petals
     Index=&SpawnGroup()
     Num=::NewEnumerator2
    )
    Mesh Scale=&MergeHelper_VectorSwitch(
     0=&Multiply_VectorByFloat(Vector=<1.15 1.15 0.2>, Float=&RandomRangeFloat(Minimum=0.25, Maximum=0.375))
     1=<0.07 0.1 0.07>
     2=?
     3=?
     Index=&SpawnGroup()
     Num=::NewEnumerator0
    )
    Mesh Scale Mode=::NewEnumerator4
    Position Offset=&MergeHelper_VectorSwitch(0=<0 0 0>, 1=<-22.4762 -16.1416 0>, 2=?, 3=?, Index=$Particles.PartitionParticles.Partition, Num=::NewEnumerator0)
    Position Offset Coordinate Space=::Local
    Sprite Size Mode=::NewEnumerator3
    Uniform Sprite Size=22.5164
    UsePositionOffset=true
    Write Material Random=true
   )
   &InitialMeshOrientation(
    Mesh Axis To Orient=<0 0 3>
    Mesh Orientation Mode=::NewEnumerator2
    Orientation Coordinate Space=::World
    Orientation Vector=&MergeHelper_VectorSwitch(0=<0 0 1>, 1=<0 1 0>, 2=?, 3=?, Index=&SpawnGroup(), Num=::NewEnumerator0)
    Rotation=&MergeHelper_VectorSwitch(
     0=&RandomRangeVector(Minimum=<0.03 0.03 0>, Maximum=<-0.03 -0.03 1>)
     1=&RandomRangeVector(Minimum=<0 0 0>, Maximum=<0 1 0>)
     2=?
     3=?
     Index=&SpawnGroup()
     Num=::NewEnumerator0
    )
    Use Rotation Vector=true
   )
   /SetVariables(Particles.DecalFacing:Vector3f=<0 0 1>) {isSetParameters=true, module=SetVariables_#7}
   &AddVelocity(
    Cone Axis=<0.1 0 1>
    Module.UseConeFalloff=false
    Module.UseVelDistribution=false
    Override Seed=false
    Velocity Mode=::NewEnumerator2
    Velocity Speed=&MaskFloatBySpawnGroup(Masked Float=&RandomRangeFloat(Minimum=50, Maximum=150), Passthrough Float=1, Spawn Group Mask=3)
   )
   &ShapeLocation(
    Cylinder Height=0
    Cylinder Radius=10
    MaskSpawnGroup=true
    Override Seed=false
    Shape Primitive=::NewEnumerator1
    Spawn Group Mask=3
    Surface Only=false
    UseOwnerScale=true
   )
  ]
  particleUpdateScript [
   &ParticleState()
   &ScaleSpriteSize(Uniform Curve Sprite Scale=DataInterface() {propertyValues {Curve {
       defaultValue=0
       keys [
        {arriveTangent=4.1785, interpMode=RCIM_Cubic, leaveTangent=4.1785, tangentMode=RCTM_User, time=0, value=0}
        {arriveTangent=1.1179, interpMode=RCIM_Cubic, leaveTangent=1.1179, tangentMode=RCTM_User, time=0.2042, value=0.6228}
        {arriveTangent=0.0224, interpMode=RCIM_Cubic, leaveTangent=0.0224, tangentMode=RCTM_User, time=1, value=1}
       ]
      }}})
   &ScaleMeshSize(Scale Factor=&MergeHelper_VectorSwitch(
     0=&VectorFromFloat(Value=&FloatFromCurve(FloatCurve=DataInterface() {propertyValues {Curve {keys [<at 0 0.4137 RCIM_Cubic>,<at 1 0.7306>]}}}))
     1=&VectorFromFloat(Value=&FloatFromCurve(FloatCurve=DataInterface() {propertyValues {Curve {keys [<at 0 0.7>,<at 0.45 0.875>,<at 1 1.5>]}}}, Scale Curve=1.25))
     2=?
     3=?
     Index=&SpawnGroup()
     Num=::NewEnumerator0
    ))
   &UpdateMeshOrientation(
    Rotation Rate=&FloatFromCurve(
     FloatCurve=DataInterface() {propertyValues {Curve {keys [<at 0.55 0.4376 RCIM_Cubic>, {arriveTangent=2.9123, interpMode=RCIM_Cubic, leaveTangent=2.9123, tangentMode=RCTM_User, time=1, value=1}]}}}
     CurveIndex=$Emitter.NormalizedLoopAge
     Scale Curve=-1.6154
    )
    Rotation Vector=&MergeHelper_VectorSwitch(0=<0 0 1>, 1=<0 3 0>, 2=?, 3=?, Index=&SpawnGroup(), Num=::NewEnumerator0)
   )
   &ScaleColor(
    Scale Alpha=&FloatFromCurve(
     FloatCurve=DataInterface() {propertyValues {Curve {keys [
         <at 0 0 RCIM_Cubic>
         {interpMode=RCIM_Cubic, tangentMode=RCTM_User, time=0.3258, value=1}
         {interpMode=RCIM_Cubic, tangentMode=RCTM_User, time=0.6888, value=1}
         {tangentMode=RCTM_User, time=1, value=0}
        ]}}}
     Scale Curve=&Multiply_Float(A=0.09, B=&FNTimeOfDayScalar(Night Scale=0.7, Morning Scale=0.9, Evening Scale=0.8))
    )
    Scale RGB=<2 2 2>
    ScaleA=true
    ScaleRGB=true
   )
   &DynamicMaterialParameters(
    Index 0 Param 1=&MergeHelper_FloatSwitch(
     0=&FloatFromCurve(FloatCurve=DataInterface() {propertyValues {Curve {keys [{tangentMode=RCTM_User, time=0.4, value=1}, {tangentMode=RCTM_User, time=0.9, value=-1}]}}})
     1=&FloatFromCurve(FloatCurve=DataInterface() {propertyValues {Curve {keys [
          {arriveTangent=2.4536, interpMode=RCIM_Cubic, leaveTangent=2.4536, tangentMode=RCTM_User, time=0, value=0}
          {interpMode=RCIM_Cubic, tangentMode=RCTM_User, time=0.5, value=0.5}
          {arriveTangent=-2.5865, tangentMode=RCTM_User, time=1, value=0}
         ]}}})
     2=?
     3=?
     Index=&SpawnGroup()
     Num=::NewEnumerator0
    )
    Index 0 Param 2=1
    Index 0 Param 3=&RandomRangeFloat(Minimum=-1, Maximum=1)
    Index 0 Param 4=&RandomRangeFloat(Minimum=-1, Maximum=1)
    Index 1 Param 1=&FloatFromCurve(
     FloatCurve=DataInterface() {propertyValues {Curve {keys [
         {arriveTangent=1.7366, interpMode=RCIM_Cubic, leaveTangent=1.7366, tangentMode=RCTM_User, time=0, value=0}
         {arriveTangent=0.4174, interpMode=RCIM_Cubic, leaveTangent=0.4174, tangentMode=RCTM_User, time=1, value=1}
        ]}}}
     Scale Curve=0.5
    )
    Index 1 Param 2=1
    Module.Index0Float0Write=true
    Module.Index0Float1Write=true
    Module.Index0Float2Write=true
    Module.Index0Float3Write=true
    Module.Index1Float0Write=true
    Module.Index1Float1Write=false
    Module.Index1Float2Write=false
    Module.Index1Float3Write=false
    Write Parameter Index 1=::NewEnumerator1
   ) {module=DynamicMaterialParameters001}
   /SetVariables(Particles.PetalSize:Vector2f=&Multiply_Vector2D_ByFloat(
     Float=&FloatFromCurve(
      FloatCurve=DataInterface() {propertyValues {Curve {
         defaultValue=1
         keys [
          {arriveTangent=14.5835, interpMode=RCIM_Cubic, leaveTangent=14.5835, tangentMode=RCTM_User, time=0, value=0}
          {arriveTangent=1.1111, interpMode=RCIM_Cubic, leaveTangent=1.1111, time=0.1, value=1}
          {arriveTangent=-1.1111, interpMode=RCIM_Cubic, leaveTangent=-1.1111, time=0.9, value=1}
          {arriveTangent=-8.7113, arriveTangentWeight=0.0333, interpMode=RCIM_Cubic, leaveTangent=-8.7113, tangentMode=RCTM_User, time=1, value=0}
         ]
        }}}
      Scale Curve=&RandomRangeFloat(Minimum=2, Maximum=3)
     )
     Vector2D=<1 1>
    )) {isSetParameters=true, module=SetVariables_#8}
   &AerodynamicDrag(
    Aerodynamic Drag:NiagaraFloat=&RandomRangeFloat(Minimum=0.3, Maximum=0.5)
    Aerodynamic Rotational Drag=0.15
    Drag Shape Curve=DataInterface() {propertyValues {Curve {keys [
        {interpMode=RCIM_Cubic, tangentMode=RCTM_User, time=0, value=0.1}
        {arriveTangent=2.0799, interpMode=RCIM_Cubic, leaveTangent=2.0799, tangentMode=RCTM_User, time=1, value=1}
       ]}}}
    Lift Contribution=0.75
    Lift Shape Curve=DataInterface() {propertyValues {Curve {keys [
        {interpMode=RCIM_Cubic, tangentMode=RCTM_User, time=0, value=0}
        {interpMode=RCIM_Cubic, tangentMode=RCTM_User, time=0.5, value=1}
        {tangentMode=RCTM_User, time=1, value=0}
       ]}}}
    Pivot Offset Maximum:Vector3f=<1 0.5 1>
    Pivot Offset Minimum:Vector3f=<-1 -1 -1>
    Rotation Strength:NiagaraFloat=&RandomRangeFloat(Minimum=1.8, Maximum=3.2)
    Use Linear Drag=true
    Use Rotational Drag=true
   )
   &SolveForcesAndVelocity(Clamp Velocity=false, Limit Acceleration=false)
   &AlignSpriteToMeshOrientation(SetAlignment=true, SetFacing=true)
  ]
  rendererClasses [&NiagaraMeshRendererProperties, &NiagaraSpriteRendererProperties]
  renderers [
   {rendererClass=&NiagaraMeshRendererProperties, rendererIndex=0}
   {rendererClass=&NiagaraMeshRendererProperties, rendererIndex=1}
   {rendererClass=&NiagaraSpriteRendererProperties, rendererIndex=2}
   {rendererClass=&NiagaraSpriteRendererProperties, rendererIndex=3}
  ]
 }
 Petals {
  emitterUpdateScript [
   &EmitterState(
    Inactive Response=::NewEnumerator1
    Life Cycle Mode=::NewEnumerator1
    Loop Behavior=::NewEnumerator1
    Loop Delay=0.3
    Loop Duration=8
    UseLoopDelay=true
   )
   &SpawnRate(Spawn Probability=0.8, SpawnRate=&RandomRangeFloat(Minimum=10, Maximum=30), Use Spawn Probability=true)
  ]
  particleSpawnScript [
   &InitializeParticle(
    Lifetime Max=3
    Lifetime Min=2
    Lifetime Mode=::NewEnumerator1
    Mass Max=1.5
    Mass Mode=::NewEnumerator3
    Position Offset=<-40.9585 0 63.7918>
    Sprite Rotation Mode=::NewEnumerator2
    Sprite Size Mode=::NewEnumerator1
    Uniform Sprite Size Max=4
    Uniform Sprite Size Min=3
    UsePositionOffset=true
    Write Material Random=true
   )
   &ShapeLocation(MaskSpawnGroup=false, Override Seed=false, Sphere Radius=65, UseOwnerScale=true)
   &InitialMeshOrientation(Use Rotation Vector=false)
  ]
  particleUpdateScript [
   &ParticleState()
   &GravityForce(Gravity=<0 0 -750>)
   &WindForce(
    Offset By Time:Vector3f=<50 10 -125>
    Scale Min / Max:Vector2f=<0.275 1>
    Speed Min / Max:Vector2f=<0 500>
    Turbulence Scale Mode:ENiagaraWindTurbulenceContributionMode=::NewEnumerator1
    Wind Speed Scale=&RandomRangeFloat(Minimum=0.2, Maximum=0.7)
   )
   &VortexForce(
    UseOriginOffset=false
    Vortex Force Amount=&FloatFromCurve(
     FloatCurve=DataInterface() {propertyValues {Curve {
        defaultValue=0
        keys [
         {interpMode=RCIM_Cubic, tangentMode=RCTM_User, time=0.7, value=0}
         {arriveTangent=3.9004, interpMode=RCIM_Cubic, leaveTangent=3.9004, tangentMode=RCTM_User, time=0.7492, value=1}
         {arriveTangent=-3.4287, interpMode=RCIM_Cubic, leaveTangent=-3.4287, tangentMode=RCTM_User, time=0.7809, value=1}
         {tangentMode=RCTM_User, time=0.8207, value=0}
        ]
       }}}
     CurveIndex=$Emitter.NormalizedLoopAge
     Scale Curve=-4008.7615
    )
   )
   &AerodynamicDrag(
    Aerodynamic Drag:NiagaraFloat=&RandomRangeFloat(Minimum=0.3333, Maximum=1)
    Aerodynamic Rotational Drag=0.15
    Lift Contribution=0.75
    Pivot Offset Maximum:Vector3f=<5 0.5 1>
    Pivot Offset Minimum:Vector3f=<-5 -0.5 -1>
    Rotation Strength:NiagaraFloat=&RandomRangeFloat(Minimum=1.8, Maximum=3.2)
    Use Linear Drag=true
    Use Rotational Drag=true
   )
   &Collision(
    Age Colliding Particles=false
    Collision Enabled=&Boolean_OrOperation(
     Bool1=&Boolean_OrOperation(
      Bool1=&IsPlatformSetActive(New Platform Set=DataInterface() {propertyValues {Platforms {qualityLevelMask=-8}}})
      Bool2=$User.bIsFrontEndPreview
     )
     Bool2=false
    )
    Friction=2
    Friction During a Bounce=0.1
    KillOnCollision=false
    Module.Control Rest State Via Surface Normal=false
    Module.EnableMaxCollisionCount=false
    Module.Randomize Collision Normal Vector=false
    Radius Calculation Type=::NewEnumerator1
    Restitution=0.05
   ) {module=Collision002}
   &SolveForcesAndVelocity(Clamp Velocity=false, Limit Acceleration=false)
   &ScaleSpriteSize(
    Uniform Curve Scale=&FloatFromCurve(FloatCurve=DataInterface() {propertyValues {Curve {keys [<at 0.85 1>,<at 1 0>]}}}, CurveIndex=$Emitter.NormalizedLoopAge)
    Uniform Curve Sprite Scale=DataInterface() {propertyValues {Curve {keys [
        {arriveTangent=14.5835, interpMode=RCIM_Cubic, leaveTangent=14.5835, tangentMode=RCTM_User, time=0, value=0}
        {arriveTangent=1.1111, interpMode=RCIM_Cubic, leaveTangent=1.1111, time=0.1, value=1}
        {arriveTangent=-1.1111, interpMode=RCIM_Cubic, leaveTangent=-1.1111, time=0.9, value=1}
        {arriveTangent=-8.7113, interpMode=RCIM_Cubic, leaveTangent=-8.7113, tangentMode=RCTM_User, time=1, value=0}
       ]}}}
   )
   &AlignSpriteToMeshOrientation(SetAlignment=true, SetFacing=true)
  ]
  rendererClasses [&NiagaraSpriteRendererProperties]
  renderers [{rendererClass=&NiagaraSpriteRendererProperties}]
 }
 Reveal001 {
  emitterUpdateScript [
   &EmitterState(Life Cycle Mode=::NewEnumerator1, Loop Behavior=::NewEnumerator1, Loop Delay=0.8, Loop Duration=6, UseLoopDelay=true)
   &SpawnRate(
    SpawnGroup=1
    SpawnRate=&SelectFloatFromArray(Direct Array Index=$Engine.QualityLevel, Float Selection Array=DataInterface() {propertyValues {FloatData [30,30,60,60,60]}})
    Use Spawn Probability=false
   ) {module=SpawnRate001}
   &SpawnRate(
    SpawnRate=&SelectFloatFromArray(Direct Array Index=$Engine.QualityLevel, Float Selection Array=DataInterface() {propertyValues {FloatData [30,30,66.7159,66.7159,66.7159]}})
    Use Spawn Probability=false
   )
   &SpawnBurst_Instantaneous(Spawn Group=2, Use Spawn Probability=false, UseLoopCountLimit=false)
   &SpawnRate(
    SpawnGroup=3
    SpawnRate=&SelectFloatFromArray(Direct Array Index=$Engine.QualityLevel, Float Selection Array=DataInterface() {propertyValues {FloatData [15,15,30,30,30]}})
    Use Spawn Probability=false
   ) {module=SpawnRate002}
  ]
  particleSpawnScript [
   /SetVariables(Particles.VisibilityTag=&SpawnGroup()) {isSetParameters=true}
   &InitializeParticle(
    Color=&MergeHelper_ColorSwitch(
     0=&SelectLinearColorFromArray(
      Array Sampling Mode=::NewEnumerator1
      Direct Array Index=$User.VariantColorIndex
      Color Selection Array=DataInterface() {propertyValues {ColorData [<rgba 3 1.4633 2.7782 0.1>, <rgba 0.3073 0.3073 0.3073 0.5>, <rgba 0 1.0502 3 0.1>]}}
      Random Seed=?
     )
     1=&SelectLinearColorFromArray(
      Array Sampling Mode=::NewEnumerator1
      Direct Array Index=$User.VariantColorIndex
      Color Selection Array=DataInterface() {propertyValues {ColorData [<rgba 3 1.4633 2.7782 0.1>, <rgba 0.3073 0.3073 0.3073 0.5>, <rgba 0 1.0502 3 0.1>]}}
      Random Seed=?
     )
     2=<rgba 0 0 0 1>
     3=&SelectLinearColorFromArray(
      Array Sampling Mode=::NewEnumerator1
      Direct Array Index=$User.VariantColorIndex
      Color Selection Array=DataInterface() {propertyValues {ColorData [<rgba 3 1.4633 2.7782 0.1>, <rgba 0.3073 0.3073 0.3073 0.5>, <rgba 0 1.0502 3 0.1>]}}
      Random Seed=?
     )
     Index=&SpawnGroup()
     Num=::NewEnumerator2
    )
    Lifetime=&MergeHelper_FloatSwitch(
     0=&RandomRangeFloat(Minimum=1, Maximum=1.5)
     1=&RandomRangeFloat(Minimum=1.5, Maximum=2)
     2=5
     3=?
     Index=&SpawnGroup()
     Num=::NewEnumerator1
    )
    Mass Max=1.5
    Mass Mode=::NewEnumerator3
    Position Mode=::NewEnumerator1
    Sprite Rotation Mode=::NewEnumerator2
    Sprite Size Mode=::NewEnumerator3
    Uniform Sprite Size=&MergeHelper_FloatSwitch(
     0=&RandomRangeFloat(Minimum=50, Maximum=100)
     1=&RandomRangeFloat(Minimum=0.5, Maximum=1)
     2=148.6187
     3=?
     Index=&SpawnGroup()
     Num=::NewEnumerator1
    )
    UsePositionOffset=true
    Write Material Random=true
   )
   &InitialMeshOrientation(Use Rotation Vector=false)
   &AddVelocity(Velocity=&MaskVectorBySpawnGroup(
     Masked Vector=&RandomRangeVector(Minimum=<0 0 5>, Maximum=<0 0 30>)
     Passthrough Vector=<1 1 0>
     Spawn Group Mask=1
    ))
   &ShapeLocation(
    Large Radius=40.3887
    MaskSpawnGroup=false
    Non Uniform Scale=<1 1 -2.2019>
    Offset=&MaskVectorBySpawnGroup(Masked Vector=<0 -11.1214 11.2861>, Passthrough Vector=<0 -5.5972 6.4401>, Spawn Group Mask=3)
    Offset Coordinate Space=::Local
    Offset Mode=::NewEnumerator0
    Override Seed=false
    Rotation Mode=::NewEnumerator2
    Shape Origin=/Niagara/DynamicInputs/Helpers/SimulationPosition()
    Shape Primitive=::NewEnumerator3
    U Distribution=0.4167
    UseOwnerScale=true
    V Distribution=0.6071
    Yaw / Pitch / Roll=<0.096 0 0>
   ) {module=ShapeLocation001}
  ]
  particleUpdateScript [
   &ParticleState()
   &CurlNoiseForce(
    Bias Noise=false
    Noise Frequency=15
    Noise Strength=&MaskFloatBySpawnGroup(Masked Float=0, Passthrough Float=10, Spawn Group Mask=2)
    Pan Noise=false
    Pan Noise Field=<976.2582 0 500>
   )
   &VortexForce(
    Influence Falloff Radius=2661.1663
    Use Influence Falloff=true
    UseOriginOffset=false
    Vortex Axis Coordinate Space:ENiagaraCoordinateSpace=::Local
    Vortex Force Amount=&MaskFloatBySpawnGroup(
     Masked Float=0
     Passthrough Float=&FloatFromCurve(
      FloatCurve=DataInterface() {propertyValues {Curve {defaultValue=0, keys [{interpMode=RCIM_Cubic, tangentMode=RCTM_User, time=0.9, value=0}, {tangentMode=RCTM_User, time=1, value=1}]}}}
      CurveIndex=$Emitter.NormalizedLoopAge
      Scale Curve=-317.2202
     )
     Spawn Group Mask=2
    )
    Vortex Origin:NiagaraPosition=&AddVectorToPosition(Position=$Particles.ShapeLocation001.ShapePosition, Vector=<0 3.4912 0>)
   )
   &SolveForcesAndVelocity(Clamp Velocity=false, Limit Acceleration=false)
   &ScaleSpriteSizeBySpeed(Max Scale Factor=<3 3>)
   &ScaleColor(
    Scale Alpha=&FloatFromCurve(
     FloatCurve=DataInterface() {propertyValues {Curve {keys [
         {interpMode=RCIM_Cubic, tangentMode=RCTM_Break, time=0, value=0}
         {interpMode=RCIM_Cubic, tangentMode=RCTM_Break, time=0.5, value=1}
         {interpMode=RCIM_Cubic, tangentMode=RCTM_Break, time=1, value=0}
        ]}}}
     Scale Curve=0.6
    )
    ScaleA=true
    ScaleRGB=true
   ) {module=ScaleColor001}
   /SetVariables(Particles.Linear Color=&MakeLinearColor(
     R:NiagaraFloat=&MakeFloatFromLinearColor(LinearColor=$Particles.Color, Channel:ENiagaraFloatFromLinearColorOptions=::NewEnumerator0)
     G:NiagaraFloat=&MakeFloatFromLinearColor(LinearColor=$Particles.Color, Channel:ENiagaraFloatFromLinearColorOptions=::NewEnumerator1)
     B:NiagaraFloat=&MakeFloatFromLinearColor(LinearColor=$Particles.Color, Channel:ENiagaraFloatFromLinearColorOptions=::NewEnumerator2)
     A:NiagaraFloat=&Multiply_Float(
      A=&MakeFloatFromLinearColor(LinearColor=$Particles.Color, Channel:ENiagaraFloatFromLinearColorOptions=::NewEnumerator3)
      B=&Lerp_Float(
       A=5
       B=0
       Alpha=&RemapRange(
        Input Value To Remap=&DistanceBetweenPositions(A=$Engine.Emitter.SimulationPosition, B=$Particles.Position)
        Input Min=&MaskFloatBySpawnGroup(Masked Float=43, Passthrough Float=1, Spawn Group Mask=3)
        Input Max=&MaskFloatBySpawnGroup(Masked Float=85.8807, Passthrough Float=108.461, Spawn Group Mask=3)
        Output Min=0
        Output Max=1
       )
      )
     )
    )) {isSetParameters=true}
   /SetVariables(
    Particles.RainbowColor:LinearColor=&ColorFromCurve(ColorCurve=DataInterface() {propertyValues {
       AlphaCurve {keys [<at 0 0>,<at 0.0971 0.3>,<at 0.7513 0.3>,<at 1 0>]}
       BlueCurve {keys [<at 0 1>,<at 1 1>]}
       GreenCurve {keys [<at 0 1>,<at 1 1>]}
       RedCurve {keys [<at 0 1>,<at 1 1>]}
      }})
    Particles.RainbowPos:NiagaraPosition=&AddVectorToPosition(Position=<-15.4534 0 50.3407>, Vector=<0 0 0>)
   ) {isSetParameters=true}
  ]
  rendererClasses [&NiagaraSpriteRendererProperties]
  renderers [
   {rendererClass=&NiagaraSpriteRendererProperties}
   {rendererClass=&NiagaraSpriteRendererProperties, rendererIndex=1}
   {rendererClass=&NiagaraSpriteRendererProperties, rendererIndex=2}
   {rendererClass=&NiagaraSpriteRendererProperties, rendererIndex=3}
  ]
 }
}
summary {
 systemName=NS_Emote_MuzzleSwish_Petals
 emitters [
  {emitterName=Petals, rendererClasses [&NiagaraSpriteRendererProperties]}
  {emitterName=Reveal001, rendererClasses [&NiagaraSpriteRendererProperties]}
  {emitterName=JumpSmall001, rendererClasses [&NiagaraMeshRendererProperties, &NiagaraSpriteRendererProperties]}
  {emitterName=JumpBig002, rendererClasses [&NiagaraMeshRendererProperties, &NiagaraSpriteRendererProperties]}
 ]
 userVariables [
  User.bIsFrontEndPreview {type=NiagaraBool}
  User.Scale {type=NiagaraFloat, defaultValue=0, description=""}
  User.VariantColorIndex {type=NiagaraInt32, defaultValue=-1}
 ]
}
